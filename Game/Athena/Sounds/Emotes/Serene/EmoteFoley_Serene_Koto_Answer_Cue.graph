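# Sound cue for the Serene emote's koto "answer" audio in Athena.
# A Random node picks one of four koto answer wave variations with equal weight.
# It plays at reduced volume in the emote music sound class, with emote attenuation and concurrency, and keeps playing while silent when virtualized.
# It also sends to a music audio-analysis submix.
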SoundCue /Game/Athena/Sounds/Emotes/Serene/EmoteFoley_Serene_Koto_Answer_Cue
root=Random_0
nodes [
 {id=Random_0, children [WavePlayer_0,WavePlayer_1,WavePlayer_2,WavePlayer_3], properties {weights [1,1,1,1]}}
 {id=WavePlayer_0, properties {soundWaveAssetPtr=&/Emote_Serene_Koto_Answer_01}}
 {id=WavePlayer_1, properties {soundWaveAssetPtr=&/Emote_Serene_Koto_Answer_02}}
 {id=WavePlayer_2, properties {soundWaveAssetPtr=&/Emote_Serene_Koto_Answer_03}}
 {id=WavePlayer_3, properties {soundWaveAssetPtr=&/Emote_Serene_Koto_Answer_04}}
]
properties {
 attenuationSettings=&Fort_Emotes_Athena_Att
 maxDistance=1600
 soundClassObject=&Fort_Music_Emote
 virtualizationMode=PlayWhenSilent
 volumeMultiplier=0.5
 concurrencyOverrides {maxCount=3}
 concurrencySet [/Game/Sounds/Concurrency_Rules/Athena_EmoteMusic_Conc]
 soundSubmixSends [{minSendLevel=0, soundSubmix=&Fortnite_AudioAnalysis_Music}]
}
